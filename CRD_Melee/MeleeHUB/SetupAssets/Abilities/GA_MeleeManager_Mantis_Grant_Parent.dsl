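; Parent gameplay ability in the melee manager setup that, when activated by a gameplay event, grants a configurable Gameplay Effect to its owner.
; It builds an outgoing spec from the GameplayEffectToApply variable.
; If GameplayEffectSetByCallerTag is valid, it assigns the event's magnitude to that tag as a SetByCaller value.
; It then applies the spec to the owner and ends immediately.
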(event Ability|EventActivateAbilityFromEvent
  (params (EventData :type Gameplay%20Event%20Data%20Structure%20%28by%20ref%29))
  (bind (_eventtag _instigator _target _optionalobject _optionalobject2 _contexthandle _instigatortags _targettags _eventmagnitude) (Utilities|Struct|BreakGameplayEventData EventData))
  (Variables|Default|SetGameplayEffectSpec (Ability|MakeOutgoingGameplayEffectSpec (Variables|Default|GetGameplayEffectToApply)))
  (if (GameplayTags|IsGameplayTagValid (Variables|Default|GetGameplayEffectSetByCallerTag))
    (bind _returnvalue (Ability|GameplayEffect|AssignTagSetbyCallerMagnitude (Variables|Default|GetGameplayEffectSpec) (Variables|Default|GetGameplayEffectSetByCallerTag) _eventmagnitude))
    (Variables|Default|SetGameplayEffectSpec _returnvalue)
    (else
      _))
  (Ability|ApplyGameplayEffectSpecToOwner (Variables|Default|GetGameplayEffectSpec))
  (Ability|EndAbility))
