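; GA_MeleeManager_Mantis_Remove_Parent is a gameplay ability for the Mantis melee manager that removes gameplay effects from its owner; the name marks it as a parent class.
; On activation from a gameplay event, it passes through a DoOnce that is never reset.
; It then removes effects granting the configured tag, up to the configured stack count, and ends itself.
; Tag and stack count are variables, and the event payload is unused.
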(event Ability|EventActivateAbilityFromEvent
  (params (EventData :type Gameplay%20Event%20Data%20Structure%20%28by%20ref%29))
  (Utilities|FlowControl|DoOnce false
    (:Completed
      (Ability|RemoveGameplayEffectFromOwnerWithGrantedTags (Variables|Default|GetGameplayTagToRemoveGameplayEffect) (Variables|Default|GetStacksToRemove))
      (Ability|EndAbility))))
